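# Looping, non-attenuated ambience cue for a broken-metal warehouse interior.
# A quality switch selects a metal room bed with quad rattles for the first two levels, or a brick bed with stereo rattles for the last two.
# Each bed mixes with winter night, spooky wind and randomly chosen distant metal rattling.
# Nested looping envelopes of differing lengths and random pitch modulators keep volume and pitch slowly drifting.
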SoundCue /Game/Sounds/Asteria/POI/Warehouse/Ambient_Warehouse_Broken_Metal_Interior_Loop_Cue
root=QualityLevel_0
nodes [
 {
  id=Enveloper_10
  children [WavePlayer_11]
  properties {
   bLoop=true
   bLoopIndefinitely=true
   loopEnd=100
   pitchCurve {editorCurveData {keys [<at 0 0.8>]}}
   volumeCurve {
    externalCurve=/Game/Sounds/Trans_Enveloper_Curves/BellReversed_100s.BellReversed_100S
    editorCurveData {keys [<at 0 1 RCIM_Cubic>,<at 50 0.5 RCIM_Cubic>,<at 100 1 RCIM_Cubic>]}
   }
  }
 }
 {
  id=Enveloper_11
  children [WavePlayer_10]
  properties {
   bLoop=true
   bLoopIndefinitely=true
   loopEnd=100
   pitchCurve {editorCurveData {keys [{time=0}]}}
   volumeCurve {
    externalCurve=/Game/Sounds/Trans_Enveloper_Curves/Bell_100s
    editorCurveData {keys [<at 0 0.1 RCIM_Cubic>,<at 100 1 RCIM_Cubic>,<at 200 0.1 RCIM_Cubic>]}
   }
  }
 }
 {
  id=Enveloper_12
  children [Modulator_0]
  properties {
   bLoop=true
   bLoopIndefinitely=true
   loopEnd=30
   pitchCurve {editorCurveData {keys [{interpMode=RCIM_Cubic, time=0}, <at 15.9468 0.9181 RCIM_Cubic>, {interpMode=RCIM_Cubic, time=30}]}}
   volumeCurve {editorCurveData {keys [<at 0 1>]}}
  }
 }
 {
  id=Enveloper_14
  children [Modulator_1]
  properties {
   bLoop=true
   bLoopIndefinitely=true
   loopEnd=30
   pitchCurve {editorCurveData {keys [{interpMode=RCIM_Cubic, time=0}, <at 15.9468 0.9181 RCIM_Cubic>, {interpMode=RCIM_Cubic, time=30}]}}
   volumeCurve {editorCurveData {keys [<at 0 1>]}}
  }
 }
 {
  id=Enveloper_15
  children [Enveloper_14]
  properties {
   bLoop=true
   bLoopIndefinitely=true
   loopEnd=40
   pitchCurve {editorCurveData {keys [{time=0}]}}
   volumeCurve {editorCurveData {keys [
      {arriveTangent=-0.01, interpMode=RCIM_Cubic, time=0, value=1}
      {arriveTangent=-0.0292, interpMode=RCIM_Cubic, leaveTangent=-0.0292, time=10.0549, value=0.15}
      {arriveTangent=-0.0105, interpMode=RCIM_Cubic, leaveTangent=-0.0105, tangentMode=RCTM_User, time=17.8246, value=0.4801}
      {arriveTangent=-0.0324, interpMode=RCIM_Cubic, leaveTangent=-0.0324, time=23.3371, value=0}
      {arriveTangent=0.0013, interpMode=RCIM_Cubic, leaveTangent=0.0013, tangentMode=RCTM_User, time=28, value=0.15}
      {arriveTangent=0.0708, interpMode=RCIM_Cubic, leaveTangent=0.0708, time=33.0549, value=0.3606}
      <at 40 1>
     ]}}
  }
 }
 {
  id=Enveloper_17
  children [Enveloper_12]
  properties {
   bLoop=true
   bLoopIndefinitely=true
   loopEnd=40
   pitchCurve {editorCurveData {keys [{time=0}]}}
   volumeCurve {editorCurveData {keys [
      {arriveTangent=-0.01, interpMode=RCIM_Cubic, time=0, value=1}
      {arriveTangent=-0.0292, interpMode=RCIM_Cubic, leaveTangent=-0.0292, time=9.7941, value=0.0731}
      {arriveTangent=-0.0105, interpMode=RCIM_Cubic, leaveTangent=-0.0105, tangentMode=RCTM_User, time=17.8246, value=0.4801}
      {arriveTangent=-0.0324, interpMode=RCIM_Cubic, leaveTangent=-0.0324, time=23.3371, value=0}
      {arriveTangent=0.0013, interpMode=RCIM_Cubic, leaveTangent=0.0013, tangentMode=RCTM_User, time=28, value=0.15}
      {arriveTangent=-0.0006, interpMode=RCIM_Cubic, leaveTangent=-0.0006, tangentMode=RCTM_User, time=32.8296, value=0.0757}
      <at 40 1>
     ]}}
  }
 }
 {
  id=Enveloper_4
  children [Modulator_2]
  properties {
   bLoop=true
   bLoopIndefinitely=true
   loopEnd=30
   pitchCurve {editorCurveData {keys [{interpMode=RCIM_Cubic, time=0}, <at 15.9468 0.9181 RCIM_Cubic>, {interpMode=RCIM_Cubic, time=30}]}}
   volumeCurve {editorCurveData {keys [<at 0 1>]}}
  }
 }
 {
  id=Enveloper_5
  children [Enveloper_4]
  properties {
   bLoop=true
   bLoopIndefinitely=true
   loopEnd=40
   pitchCurve {editorCurveData {keys [{time=0}]}}
   volumeCurve {editorCurveData {keys [
      {arriveTangent=-0.01, interpMode=RCIM_Cubic, time=0, value=1}
      <at 10.0039 0.2703 RCIM_Cubic>
      {arriveTangent=-0.0105, interpMode=RCIM_Cubic, leaveTangent=-0.0105, tangentMode=RCTM_User, time=20, value=1}
      {arriveTangent=-0.025, interpMode=RCIM_Cubic, leaveTangent=-0.025, time=23.5545, value=0.0703}
      {arriveTangent=0.0013, interpMode=RCIM_Cubic, leaveTangent=0.0013, tangentMode=RCTM_User, time=28, value=0.8}
      {arriveTangent=0.0167, interpMode=RCIM_Cubic, leaveTangent=0.0167, time=33.0039, value=0.2703}
      <at 40 1>
     ]}}
  }
 }
 {
  id=Enveloper_6
  children [Modulator_3]
  properties {
   bLoop=true
   bLoopIndefinitely=true
   loopEnd=30
   pitchCurve {editorCurveData {keys [{interpMode=RCIM_Cubic, time=0}, <at 15.9468 0.9181 RCIM_Cubic>, {interpMode=RCIM_Cubic, time=30}]}}
   volumeCurve {editorCurveData {keys [<at 0 1>]}}
  }
 }
 {
  id=Enveloper_7
  children [Enveloper_6]
  properties {
   bLoop=true
   bLoopIndefinitely=true
   loopEnd=40
   pitchCurve {editorCurveData {keys [{time=0}]}}
   volumeCurve {editorCurveData {keys [
      {arriveTangent=-0.01, interpMode=RCIM_Cubic, time=0, value=1}
      <at 10.0039 0.2703 RCIM_Cubic>
      {arriveTangent=-0.0105, interpMode=RCIM_Cubic, leaveTangent=-0.0105, tangentMode=RCTM_User, time=20, value=1}
      {arriveTangent=-0.025, interpMode=RCIM_Cubic, leaveTangent=-0.025, time=23.5545, value=0.0703}
      {arriveTangent=0.0013, interpMode=RCIM_Cubic, leaveTangent=0.0013, tangentMode=RCTM_User, time=28, value=0.8}
      {arriveTangent=0.0167, interpMode=RCIM_Cubic, leaveTangent=0.0167, time=33.0039, value=0.2703}
      <at 40 1>
     ]}}
  }
 }
 {
  id=Enveloper_8
  children [WavePlayer_3]
  properties {
   bLoop=true
   bLoopIndefinitely=true
   loopEnd=100
   pitchCurve {editorCurveData {keys [<at 0 0.8>]}}
   volumeCurve {
    externalCurve=&BellReversed_100S
    editorCurveData {keys [<at 0 1 RCIM_Cubic>,<at 50 0.5 RCIM_Cubic>,<at 100 1 RCIM_Cubic>]}
   }
  }
 }
 {
  id=Enveloper_9
  children [WavePlayer_1]
  properties {
   bLoop=true
   bLoopIndefinitely=true
   loopEnd=100
   pitchCurve {editorCurveData {keys [{time=0}]}}
   volumeCurve {
    externalCurve=&Bell_100s
    editorCurveData {keys [<at 0 0.1 RCIM_Cubic>,<at 100 1 RCIM_Cubic>,<at 200 0.1 RCIM_Cubic>]}
   }
  }
 }
 {id=Mixer_0, children [WavePlayer_2,Enveloper_5], properties {inputVolume [1,0.75]}}
 {id=Mixer_1, children [Enveloper_7,WavePlayer_0], properties {inputVolume [1,0.5]}}
 {id=Mixer_2, children [Enveloper_8,Enveloper_9,Enveloper_17], properties {inputVolume [0.8,0.25,0.1]}}
 {id=Mixer_3, children [Enveloper_10,Enveloper_11,Enveloper_15], properties {inputVolume [0.8,0.25,0.2]}}
 {id=Modulator_0, children [Random_1], properties {pitchMax=0.9, pitchMin=0.8}}
 {id=Modulator_1, children [Random_0], properties {pitchMax=0.9, pitchMin=0.8}}
 {id=Modulator_2, children [Mixer_2], properties {pitchMax=0.9, pitchMin=0.8}}
 {id=Modulator_3, children [Mixer_3], properties {pitchMax=0.9, pitchMin=0.8}}
 {id=QualityLevel_0, children [Mixer_0,Mixer_0,Mixer_1,Mixer_1]}
 {id=Random_0, children [WavePlayer_5,WavePlayer_7], properties {weights [1,1]}}
 {id=Random_1, children [WavePlayer_4,WavePlayer_6], properties {weights [1,1]}}
 {id=WavePlayer_0, properties {bLooping=true, soundWaveAssetPtr=/Game/Sounds/Apollo/Biomes/Interiors_Generic/AMB_Bed_Int_Brick_Med_01_St}}
 {id=WavePlayer_1, properties {bLooping=true, soundWaveAssetPtr=/Game/Sounds/Trans_AmbientLoops/Wind/WindSpooky01_Loop_Stereo}}
 {id=WavePlayer_10, properties {bLooping=true, soundWaveAssetPtr=&WindSpooky01_Loop_Stereo}}
 {id=WavePlayer_11, properties {bLooping=true, soundWaveAssetPtr=/Game/Sounds/Fort_Ambient_Looping/Winter/Amb_Winter_Night_Loop}}
 {id=WavePlayer_2, properties {bLooping=true, soundWaveAssetPtr=/Game/Sounds/Apollo/Biomes/Interiors_Generic/AMB_Bed_Int_Metal_Lg_01_Quad}}
 {id=WavePlayer_3, properties {bLooping=true, soundWaveAssetPtr=&Amb_Winter_Night_Loop}}
 {
  id=WavePlayer_4
  properties {bLooping=true, soundWaveAssetPtr=/Game/Sounds/Asteria/POI/JunkYard/Ambience_Metal_Distant_Rattling_In_Wind_Loop_03_Quad}
 }
 {id=WavePlayer_5, properties {soundWaveAssetPtr=/Game/Sounds/Asteria/POI/JunkYard/Ambience_Metal_Distant_Rattling_In_Wind_Loop_04_Stereo}}
 {
  id=WavePlayer_6
  properties {bLooping=true, soundWaveAssetPtr=/Game/Sounds/Asteria/POI/JunkYard/Ambience_Metal_Distant_Rattling_In_Wind_Loop_04_Quad}
 }
 {id=WavePlayer_7, properties {soundWaveAssetPtr=/Game/Sounds/Asteria/POI/JunkYard/Ambience_Metal_Distant_Rattling_In_Wind_Loop_03_Stereo}}
]
properties {
 attenuationSettings=&Fort_Ambient_Athena_Att
 maxDistance=2097152
 soundClassObject=&Fort_Ambient_Athena
 volumeMultiplier=0.652
 attenuationOverrides {bAttenuate=false}
}
